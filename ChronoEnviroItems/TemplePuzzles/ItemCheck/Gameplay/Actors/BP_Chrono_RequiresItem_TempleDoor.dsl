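; Temple puzzle vault door presented as requiring an item of a given rarity, which sets its Niagara color.
; On authority, it opens after a gameplay event, immediately if the hotfix flag is off, or when a character overlaps the back trigger.
; Timelines slide both halves, OnRep functions update collision and prompts, and clients play dust, shake and audio.
; Conversation messages track the local interacting controller for prompt sounds and the music stinger.
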(event OnComponentBeginOverlap(BackOverlapAndOpenBox)
  (params (OverlappedComponent :type Primitive%20Component%20Object%20Reference) (OtherActor :type Actor%20Object%20Reference) (OtherComp :type Primitive%20Component%20Object%20Reference) (OtherBodyIndex :type Integer) (bFromSweep :type Boolean) (SweepResult :type Hit%20Result%20Structure%20%28by%20ref%29))
  (Utilities|Casting|CastToCharacter OtherActor)
  (CallFunction|OpenVaultDoor)
  (Components|DestroyComponent (Variables|Default|GetBackOverlapAndOpenBox)))

(event Custom|CheckItemRequiredRarity
  ;[1 1 1] Map the rarity from enum to an Int value. Compare required rarity value for object against item map value
  (bind _output_get (Variables|Default|SetRarity%20Check%20Required%20Int (Utilities|Map|Find (Variables|Default|GetRarity%20Map%20From%20Enum%20to%20Integer%20Value) (Variables|Default|GetRarity%20Check%20Required%20Rarity))))
  (Niagara|SetNiagaraVariable(Int32) (Variables|Default|GetNS_Chrono_Temple_Door_ItemCheck) "ColorFX" _output_get))

(event EventBeginPlay
  ;[1 1 1] Init
  (Class|FortAsyncActionAuthorityMatchReady|MatchReadyAsync
    (:then
      (CallFunction|CheckItemRequiredRarity) ; This need to happen on client for FX
    )
    (:OnReady
      (CallFunction|SetStartandEndPositionsforDoors)
      (if (ScalableFloat|GetValueasBool (Variables|Default|GetHotFixEnabled))
        (Networking|SwitchHasAuthority
          (:Authority
            (Ability|Async|WaitGameplayEventToActor 0 "(TagName=\"Event.TemplePuzzle.ItemDoor.Enabled\")"
              (:EventReceived
                (Variables|Default|SetPuzzle%20Solved true)
                (if (Variables|Default|GetPuzzle%20Solved)
                  (Utilities|FlowControl|Delay (ScalableFloat|GetValueatLevel (Variables|Default|GetDelayBeforeDoorOpens)))
                  (CallFunction|OpenVaultDoor))))))
        (else
          (CallFunction|OpenVaultDoor)
          (Components|DestroyComponent (Variables|Default|GetBackOverlapAndOpenBox)))))))

(event Custom|Open
  (bind _audioloop (Variables|Default|GetAudioLoop))
  (bind _returnvalue (Transformation|GetWorldLocation _audioloop))
  ;[1 1 1] Open Door
  (Utilities|FlowControl|DoOnce false
    (:Completed
      (Variables|Default|SetDoorOpened true)
      (Utilities|FlowControl|Sequence
        (:then_0
          (Utilities|FlowControl|Sequence
            (:then_0
              (CallFunction|DoorAnimatingCosmetics) ; inc audio
              (if (Variables|Default|GetDebugResetState)
                (Utilities|FlowControl|Delay 1e+01)
                (CallFunction|CloseVaultDoor)
                (Components|Activation|Activate (Variables|Default|GetNS_Chrono_Temple_Door_ItemCheck) false)))
            (:then_1
              (bind (_direction _dooralpha) (Timeline|DoorOpenAnim PlayFromStart)
                (:Update
                  (Niagara|SetNiagaraVariable(Float) (Variables|Default|GetNS_Dust) "DoorOpen" _dooralpha)
                  (Transformation|SetWorldLocation (Variables|Default|GetSM_Ancients_VaultDoor_R) (Math|Vector|Lerp(Vector) (Variables|Default|GetStartLocation_R) (Variables|Default|GetEndLocation_R) _dooralpha) false false)
                  (Transformation|SetWorldLocation (Variables|Default|GetSM_Ancients_VaultDoor_L) (Math|Vector|Lerp(Vector) (Variables|Default|GetStartLocation_L) (Variables|Default|GetEndLocation_L) _dooralpha) false false))
                (:EndShake
                  (Audio|SpawnSoundatLocation (Variables|Sound|GetDoorStopSound) _returnvalue)
                  (Audio|SpawnSoundatLocation (Variables|Sound|GetCamShakeSound) _returnvalue)
                  (Camera|PlayWorldCameraShake (Variables|Default|GetShake) _returnvalue 2e+03 2e+03 0.75)
                  (Components|DestroyComponent (Variables|Default|GetBox_Nav_Middle)) ; Allow NavMesh to Generate in the Door Opening Now
                  (ForceFeedback|SpawnForceFeedbackatLocation (Variables|Default|GetForce%20Feedback%20Effect) (Transformation|GetWorldLocation (Variables|Default|GetStaticMesh))))))))
        (:then_1
          ;;[1 1 1] Destroy back box trigger once opened.
          (Components|DestroyComponent (Variables|Default|GetBackOverlapAndOpenBox)))))))

(event Custom|OpenVaultDoor
  (bind _audioloop (Variables|Default|GetAudioLoop))
  (bind _returnvalue (Transformation|GetWorldLocation _audioloop))
  ;[1 1 1] Open Door
  (Utilities|FlowControl|DoOnce false
    (:Completed
      (Variables|Default|SetDoorOpened true)
      (Utilities|FlowControl|Sequence
        (:then_0
          (Utilities|FlowControl|Sequence
            (:then_0
              (CallFunction|DoorAnimatingCosmetics) ; inc audio
              (if (Variables|Default|GetDebugResetState)
                (Utilities|FlowControl|Delay 1e+01)
                (CallFunction|CloseVaultDoor)
                (Components|Activation|Activate (Variables|Default|GetNS_Chrono_Temple_Door_ItemCheck) false)))
            (:then_1
              (Timeline|DoorOpenAnim PlayFromStart))))
        (:then_1
          ;;[1 1 1] Destroy back box trigger once opened.
          (Components|DestroyComponent (Variables|Default|GetBackOverlapAndOpenBox)))))))

(event Custom|CloseVaultDoor
  (bind _audioloop (Variables|Default|GetAudioLoop))
  (bind _returnvalue (Transformation|GetWorldLocation _audioloop))
  ;[1 1 1] Open Door
  (if (Variables|Default|GetDebugResetState)
    (Variables|Default|SetDoorOpened false)
    (else
      (Utilities|FlowControl|DoOnce false
        (:Completed
          (Variables|Default|SetDoorOpened true)))))
  (CallFunction|DoorAnimatingCosmetics)
  (bind (_direction _dooralpha) (Timeline|Timeline PlayFromStart)
    (:Update
      (Niagara|SetNiagaraVariable(Float) (Variables|Default|GetNS_Dust) "DoorOpen" _dooralpha)
      (Transformation|SetWorldLocation (Variables|Default|GetSM_Ancients_VaultDoor_R) (Math|Vector|Lerp(Vector) (Variables|Default|GetEndLocation_R) (Variables|Default|GetStartLocation_R) _dooralpha) false false)
      (Transformation|SetWorldLocation (Variables|Default|GetSM_Ancients_VaultDoor_L) (Math|Vector|Lerp(Vector) (Variables|Default|GetEndLocation_L) (Variables|Default|GetStartLocation_L) _dooralpha) false false))
    (:EndShake
      (Audio|SpawnSoundatLocation (Variables|Sound|GetDoorStopSound) _returnvalue)
      (Audio|SpawnSoundatLocation (Variables|Sound|GetCamShakeSound) _returnvalue))) ; door closing anim
)

(event Custom|DoorAnimatingCosmetics
  (bind _audioloop (Variables|Default|GetAudioLoop))
  (bind _returnvalue (Transformation|GetWorldLocation _audioloop))
  ;[1 1 1] FX and audio
  (Networking|SwitchHasAuthority
    (:Remote
      (Utilities|FlowControl|Sequence
        (:then_0
          (Camera|PlayWorldCameraShake (Variables|Default|GetShake) (Transformation|GetWorldLocation (Variables|Default|GetStaticMesh)) 2e+03 2e+03 0.75)
          (bind _returnvalue_2 (Niagara|SpawnSystemAttached "/ChronoEnviroItems/TemplePuzzles/RedLight/FX/NS_Chrono_Temple_DoorDust.NS_Chrono_Temple_DoorDust" (Variables|Default|GetStaticMesh) "None" (Transformation|GetWorldLocation (Variables|Default|GetStaticMesh)) "0, 0, 0" "KeepWorldPosition"))
          (Variables|Default|SetNS_Dust _returnvalue_2)
          (ForceFeedback|SpawnForceFeedbackatLocation (Variables|Default|GetForce%20Feedback%20Effect) (Transformation|GetWorldLocation (Variables|Default|GetStaticMesh))))
        (:then_1
          ;;[0.05153 0.09729 0.1969] Audio
          (Utilities|FlowControl|DoOnce false
            (:Completed
              (Audio|SpawnSoundatLocation (Variables|Sound|GetDoorOpenSound) _returnvalue)
              (Audio|SpawnSoundatLocation (Variables|Sound|GetDoorMovingSound) _returnvalue "0.000000,180.000000,0.000000")
              (if (Variables|Default|GetDoorOpened)
                (bind _audioloop_2 (Variables|Default|GetAudioLoop)
                  (:then
                    (Audio|Components|Audio|FadeOut _audioloop_2 7.0 0.0 "Linear" :!class /Script/Engine.AudioComponent)
                    (Utilities|FlowControl|Delay 2.0)
                    (Utilities|FlowControl|Sequence
                      (:then_0
                        _)
                      (:then_1
                        _))
                    (Utilities|FlowControl|Delay 8.0)
                    (Utilities|FlowControl|DoOnce false
                      (:Completed
                        (Utilities|IsValid (Variables|Default|GetLocal%20Interacting%20Controller)
                          (:Is%20Valid
                            (Audio|SpawnSound2D (Variables|Sound|GetMusicStinger))
                            (Utilities|FlowControl|Delay 1.0))) ; Having this variable set means we're in the client that started interacting with this door.
                      )))
                  (:else
                    _)))
              (bind _audioloop_3 (Variables|Default|GetAudioLoop)
                (:then
                  (Audio|Components|Audio|FadeIn _audioloop_3 7.0 1.0 0.0 "Linear" :!class /Script/Engine.AudioComponent))
                (:else
                  _)))))))))

(event Custom|DestroyDoor
  (bind _self self)
  ;[1 1 1] Destroy
  (Variables|Default|SetDoorDestroyed true)
  (Networking|SwitchHasAuthority
    (:Authority
      (Utilities|IsValid _self
        (:Is%20Valid
          (Actor|DestroyActor _self))))
    (:Remote
      (Niagara|SpawnSystematLocation "/Game/Effects/Fort_Effects/World/Destruction/Base_Effects/NS_General_Deco_Destruction_Large_Converted.NS_General_Deco_Destruction_Large_Converted" (Transformation|GetWorldLocation (Variables|Default|GetDefaultSceneRoot)))
      (Rendering|SetActorHiddenInGame self true)
      (bind _audioloop (Variables|Default|GetAudioLoop)
        (:then
          (Components|DestroyComponent _audioloop))
        (:else)))))

(event OnReceivedLocalClientConversationMessage(NonPlayerConversationComponent)
  (params (LocalPlayerController :type Fort%20Player%20Controller%20Object%20Reference) (Message :type Client%20Conversation%20Message%20Payload%20Structure%20%28by%20ref%29))
  ;[0.05153 0.09729 0.1969] Audio For Text Prompts
  (Networking|SwitchHasAuthority
    (:Remote
      (bind _output_get (Variables|Default|SetLocal%20Interacting%20Controller LocalPlayerController))
      (Audio|SpawnSound2D (Variables|Sound|GetTextPromptOpenSound))
      (Ability|Tasks|WaitGameplayTagRemoveFromActor (Variables|/Script/Engine.Controller|GetPlayerState _output_get) "(TagName=\"Conversation.InteractingWithNPC\")" true
        (:Removed
          (CallFunction|OnClientFinishedConversation))))))

(event Custom|OnClientFinishedConversation
  ;[0.05153 0.09729 0.1969] Audio For Text Prompts
  (Utilities|FlowControl|Sequence
    (:then_0
      (Audio|SpawnSound2D (Variables|Sound|GetTextPromptCloseSound)))
    (:then_1
      (Utilities|Time|ClearandInvalidateTimerbyHandle (Variables|Default|GetClientFinishesConversationTimer))
      (bind _returnvalue (Utilities|Time|SetTimerbyEvent (delegate Custom|CheckIfDoorOpenToUnsetController) 2.0))
      (Variables|Default|SetClientFinishesConversationTimer _returnvalue))))

(event Custom|CheckIfDoorOpenToUnsetController
  ;[0.05153 0.09729 0.1969] Audio For Text Prompts
  (if (not (Variables|Default|GetDoorOpened))
    (Variables|Default|SetLocal%20Interacting%20Controller 0)))

(fn BlueprintCanInteract
  (params (InteractingPawn :type Fort%20Pawn%20Object%20Reference) (InteractionBeingAttempted :type EInteractionBeingAttempted%20Enum :default "FirstInteraction") (InteractionType :type TInteractionType%20Enum :default "IT_NoInteraction"))
  (return true))

(fn BlueprintGetAllHighlightableComponents
  (return (Utilities|Array|MakeArray (Variables|Default|GetSM_Ancients_VaultDoor_L) (Variables|Default|GetSM_Ancients_VaultDoor_R)) true))

(fn BlueprintGetFailedInteractionString
  (params (InteractingPawn :type Fort%20Pawn%20Object%20Reference) (InteractionBeingAttempted :type EInteractionBeingAttempted%20Enum :default "FirstInteraction"))
  (return ""))

(fn BlueprintGetInteractionString
  (params (InteractingPawn :type Fort%20Pawn%20Object%20Reference) (InteractionBeingAttempted :type EInteractionBeingAttempted%20Enum :default "FirstInteraction"))
  (return (Variables|Default|GetInteraction%20String)))

(fn BlueprintGetInteractionTime
  (params (InteractingPawn :type Fort%20Pawn%20Object%20Reference) (InteractionBeingAttempted :type EInteractionBeingAttempted%20Enum :default "FirstInteraction"))
  (return 0.0 true))

(fn OnRep_DoorDestroyed
  (if (Variables|Default|GetDoorDestroyed)
    (Rendering|SetVisibility (Variables|Default|GetStaticMesh) false false)
    (Collision|SetCollisionEnabled (Variables|Default|GetBlockAllCollision) "NoCollision")
    (else
      (Rendering|SetVisibility (Variables|Default|GetStaticMesh) true false))))

(fn OnRep_DoorOpened
  (if (Variables|Default|GetDoorOpened)
    (Collision|SetCollisionEnabled (Variables|Default|GetBlockAllCollision) "NoCollision")
    (InteractionUI|AddHideUIIndicatorReason (Variables|Default|GetNonPlayerConversationComponent) (Variables|Default|GetHide%20Conversation%20UI%20Indicator%20Reason))
    (Variables|Interaction|SetbShowFirstInteractPrompt false)
    (Components|Activation|Deactivate (Variables|Default|GetNS_Chrono_Temple_Door_ItemCheck))
    (else
      (Collision|SetCollisionEnabled (Variables|Default|GetBlockAllCollision) "QueryAndPhysics"))))

(fn SetStartAndEndPositionsForDoors
  (bind _sm_ancients_vaultdoor_l (Variables|Default|GetSM_Ancients_VaultDoor_L))
  (bind _returnvalue (Transformation|GetWorldLocation _sm_ancients_vaultdoor_l))
  (bind _sm_ancients_vaultdoor_r (Variables|Default|GetSM_Ancients_VaultDoor_R))
  (bind _returnvalue_2 (Transformation|GetWorldLocation _sm_ancients_vaultdoor_r))
  ;[1 1 1] Store off base 'start' position. Right door needs to go 'backwards' negative units
  (Variables|Default|SetStartLocation_R _returnvalue_2)
  (Variables|Default|SetEndLocation_R (+ _returnvalue_2 (* (Transformation|GetForwardVector _sm_ancients_vaultdoor_r) (Math|Conversions|ToVector(Float) (Variables|Default|GetNegativeDoorExtensionLength)))))
  ;[1 1 1] Store off base 'start' position. Left door needs to go 'forwards' positive units
  (Variables|Default|SetStartLocation_L _returnvalue)
  (Variables|Default|SetEndLocation_L (+ _returnvalue (* (Transformation|GetForwardVector _sm_ancients_vaultdoor_l) (Math|Conversions|ToVector(Float) (Variables|Default|GetDoorExtensionLength)))))
  (return))

(fn ConstructionScript)
