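; Interactable conversation actor for the temple item-check door puzzle, relaying the door-enable gameplay event to its linked Temple Door.
; On BeginPlay the server waits out a fixed delay, marked as a TODO for async loading, then listens for the ItemDoor.Enabled event and resends it to the door.
; Interaction overrides return variable-driven availability and prompt text, with zero interaction time.
; The interact, overlap and tick events are empty here.
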(event EventBeginPlay
  (Networking|SwitchHasAuthority
    (:Authority
      ;[1 0.9595 0.07245] TODO async load wait?
      (Utilities|FlowControl|Delay 1e+01)
      (Ability|Async|WaitGameplayEventToActor 0 "(TagName=\"Event.TemplePuzzle.ItemDoor.Enabled\")"
        (:EventReceived
          (Ability|SendGameplayEventtoActor (Variables|Default|GetTemple%20Door) "(TagName=\"Event.TemplePuzzle.ItemDoor.Enabled\")"))))))

(event Collision|EventActorBeginOverlap
  (params (OtherActor :type Actor%20Object%20Reference)))

(event EventTick
  (params (DeltaSeconds :type Float%20%28single-precision%29)))

(event Interaction|EventBlueprintOnInteract
  (params (InteractingPawn :type Fort%20Pawn%20Object%20Reference) (InteractionBeingAttempted :type EInteractionBeingAttempted%20Enum :default "FirstInteraction")))

(fn BlueprintCanInteract
  (params (InteractingPawn :type Fort%20Pawn%20Object%20Reference) (InteractionBeingAttempted :type EInteractionBeingAttempted%20Enum :default "FirstInteraction") (InteractionType :type TInteractionType%20Enum :default "IT_NoInteraction"))
  (return (Variables|Default|GetCan%20Interact)))

(fn BlueprintGetFailedInteractionString
  (params (InteractingPawn :type Fort%20Pawn%20Object%20Reference) (InteractionBeingAttempted :type EInteractionBeingAttempted%20Enum :default "FirstInteraction"))
  (return ""))

(fn BlueprintGetInteractionString
  (params (InteractingPawn :type Fort%20Pawn%20Object%20Reference) (InteractionBeingAttempted :type EInteractionBeingAttempted%20Enum :default "FirstInteraction"))
  (return (Variables|Default|GetInteraction%20String)))

(fn BlueprintGetInteractionTime
  (params (InteractingPawn :type Fort%20Pawn%20Object%20Reference) (InteractionBeingAttempted :type EInteractionBeingAttempted%20Enum :default "FirstInteraction"))
  (return 0.0 true))

(fn ConstructionScript)
